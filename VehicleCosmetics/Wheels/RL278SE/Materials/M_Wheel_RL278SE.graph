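# Animated special-edition RL278SE vehicle wheel material, overriding base color, emissive and normal from MF_WheelParent_Simple.
# Two custom colors blend through masks, panning facet detail, sine-driven shifts and a scrolling gradient distorted by the tire normal, with vertex-color red separating rim from tire.
# A rotating, partly desaturated reflection cubemap adds fake environment shimmer, which also feeds emissive.
# The result passes through vehicle damage and de-rez transition functions.
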Material /VehicleCosmetics/Wheels/RL278SE/Materials/M_Wheel_RL278SE
graph {
 properties {
  bAutomaticallySetUsageInEditor=false
  bUseMaterialAttributes=true
  bUsedWithMeshDeformer=true
  bUsedWithSkeletalMesh=true
  pixelDepthOffsetMode=PDOM_Legacy
  previewMesh=/VehicleCosmetics/Wheels/RL240/Meshes/SK_Wheel_RL240
  usedShadingModels=Default Lit
 }
 nodes [
  Constant2Vector() {g=1}
  n1 = MaterialFunctionCall() {materialFunction=&MF_WheelParent_Simple}
  n2 = GetMaterialAttributes(None=@n1.MaterialAttributes) {attributeGetTypes [#1]}
  n3 = NamedRerouteUsage()
  n4 = Multiply(A=ComponentMask(None=@n3) {r=true}, B=ComponentMask(None=@n3) {g=true})
  n5 = VectorParameter() {group=5_Special Effects, parameterName=CustomColor1, sortPriority=34, defaultValue=<rgba 0.0802 0.0802 0.0802 1>}
  n6 = ScalarParameter() {defaultValue=0.1, group=5_Special Effects, parameterName=PatternContrast, sortPriority=43}
  n7 = Multiply(A=@n5.RGB, B=@n6) {constA=0.3, constB=0.6}
  n8 = VectorParameter() {group=5_Special Effects, parameterName=CustomColor2, sortPriority=33, defaultValue=<rgba 0.5583 0.5583 0.5583 1>}
  n9 = Multiply(A=@n8.RGB, B=@n6) {constA=0.3, constB=0.6}
  n10 = TextureCoordinate() {uTiling=2, unMirrorU=true, vTiling=0.5}
  n11 = Multiply(
   A=ScalarParameter() {defaultValue=0.5, group=5_Special Effects, parameterName=DetailTiling, sortPriority=42}
   B=ComponentMask(None=@n10) {r=true}
  )
  n12 = ComponentMask(None=@n10) {g=true}
  n13 = TextureSampleParameter2D(UVs=Panner(Coordinate=AppendVector(A=@n11, B=@n12)) {speedX=0.02}) {
   group=5_Special Effects
   parameterName=DetailNormal
   samplerType=SAMPLERTYPE_Normal
   sortPriority=35
   texture=/VehicleCosmetics/Wheels/RL278SE/Textures/Detail_polynormal_N
  }
  n14 = LinearInterpolate(A=LinearInterpolate(A=@n7, B=@n5.RGB, Alpha=Clamp(None=@n4)), B=LinearInterpolate(A=@n9, B=@n8.RGB, Alpha=@n13.B))
  n15 = Multiply(
   A=Multiply(A=@n14, B=@n14)
   B=ScalarParameter() {defaultValue=0.1, group=5_Special Effects, parameterName=LighterRimIntensity, sortPriority=41}
  )
  n16 = TextureCoordinate()
  n17 = AppendVector(
   A=Multiply(
    A=ScalarParameter() {defaultValue=1.2, group=5_Special Effects, parameterName=Tire Tiling, sortPriority=38}
    B=ComponentMask(None=@n16) {r=true}
   )
   B=ComponentMask(None=@n16) {g=true}
  )
  n18 = TextureSampleParameter2D(UVs=@n17) {
   group=5_Special Effects
   parameterName=TireNormal
   samplerType=SAMPLERTYPE_LinearColor
   sortPriority=37
   texture=/VehicleCosmetics/Wheels/RL278/Textures/T_Tire_RL278_N
  }
  n19 = Clamp(None=Add(A=@n18.R, B=@n18.R))
  n20 = Add(A=@n19, B=@n19)
  n21 = LinearInterpolate(A=@n7, B=@n5.RGB, Alpha=@n20)
  n22 = LinearInterpolate(A=@n9, B=@n8.RGB, Alpha=@n20)
  n23 = Sine(None=Multiply(A=ScalarParameter() {defaultValue=0.2, group=5_Special Effects, parameterName=Rim_ShiftSpeed, sortPriority=39}, B=Time()))
  n24 = Add(A=Power(Base=NamedRerouteUsage()) {constExponent=3}, B=Divide(A=Add(A=@n23)))
  n25 = Multiply(A=Panner(Coordinate=TextureCoordinate()) {speedX=0.5}) {constB=0.5}
  n26 = Power(Base=@n18.B) {constExponent=10}
  n27 = Multiply(
   A=OneMinus(None=Clamp(None=@n26) {maxDefault=10})
   B=ScalarParameter() {defaultValue=0.6, group=5_Special Effects, parameterName=ShimmerShelfDistort, sortPriority=44}
  )
  n28 = LinearInterpolate(
   A=ComponentMask(None=Frac(None=@n24)) {b=true, g=true, r=true}
   B=TextureSample(UVs=Subtract(A=@n25, B=@n27)) {samplerType=SAMPLERTYPE_LinearGrayscale, texture=/VehicleCosmetics/Wheels/RL209SE/Textures/T_GradientV_T}.RGB
   Alpha=VertexColor().R
  )
  n29 = TextureSample(UVs=Rotator(Coordinate=ReflectionVectorWS()) {speed=0.5}) {samplerType=SAMPLERTYPE_LinearColor, texture=/VehicleCosmetics/Wheels/DuelingDragonsSE/Materials/TC_DuelingDragonsSE_GrassCube}
  n30 = Constant() {r=0.5}
  n31 = Add(
   A=Multiply(
    A=Desaturation(None=@n29.RGB, Fraction=@n30) {luminanceFactors=<rgba 0.3 0.59 0.11 0>}
    B=ScalarParameter() {defaultValue=1, group=5_Special Effects, parameterName=WorldEnvIntensity, sortPriority=45}
   )
   B=LinearInterpolate(
    A=StaticSwitchParameter(True=@n15, False=@n14) {group=5_Special Effects, parameterName=LighterRim, sortPriority=40}
    B=LinearInterpolate(A=@n21, B=@n22, Alpha=@n28)
    Alpha=VertexColor().R
   )
  )
  n32 = Clamp(None=@n31)
  n33 = Clamp(None=Multiply(A=@n31, B=ScalarParameter() {defaultValue=0.1, group=5_Special Effects, parameterName=EmissiveIntensity, sortPriority=41}) {constB=0.25})
  n34 = Clamp(None=LinearInterpolate(A=NamedRerouteUsage(), B=NamedRerouteUsage(), Alpha=VertexColor().R))
  MaterialFunctionCall(In=MaterialFunctionCall(Material=SetMaterialAttributes!(MaterialAttributes=@n2.MaterialAttributes, Base Color=@n32, Emissive Color=@n33, Normal=@n34) {attributeSetTypes [#2, #3, #1]}) {materialFunction=&MF_BaseVehicle_Damage}.Result) {materialFunction=&MF_VehicleDeRez_Transition}
  n35 = VertexColor()
  NamedRerouteDeclaration(None=Multiply(A=@n2.Normal, B=OneMinus(None=@n35.R))) {name=RimNormalOutput, nodeColor=<rgba 1 0.8906 0 1>}
  n36 = TextureSampleParameter2D(UVs=@n17) {
   group=5_Special Effects
   parameterName=TireFace
   samplerType=SAMPLERTYPE_LinearColor
   sortPriority=36
   texture=/VehicleCosmetics/Wheels/RL278SE/Textures/Tire_RL278SE_facets
  }
  n37 = Normalize(VectorInput=Add(A=@n36.RGB, B=Constant3Vector() {constant=<rgba 0.3 0.3 1 0>}))
  n38 = Constant3Vector() {constant=<rgba 0.5 0.5 1 0>}
  NamedRerouteDeclaration(None=MaterialFunctionCall(Base=Add(A=@n37, B=@n37), Blend=Multiply(A=@n13.RGB, B=@n38)) {materialFunction=&Blend_Overlay}.Result) {name=TireNormalOutput, nodeColor=<rgba 0 1 0.25 1>}
  NamedRerouteDeclaration(None=@n1.RimRGBAMask) {name=RimAMask, nodeColor=<rgba 1 0 0.3984 1>}
  VectorParameter() {group=5_Special Effects, parameterName=CustomColor, sortPriority=34, defaultValue=<rgba 0.0802 0.0802 0.0802 1>}
  VectorParameter() {group=5_Special Effects, parameterName=AccentColor, sortPriority=33, defaultValue=<rgba 0.5583 0.5583 0.5583 1>}
 ]
}
